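# Sound cue for bull rush footsteps tied to a scythe power-up, classed as melee weapon audio.
# A random node picks one of five equally weighted Constructor Bullrush step waves, then a modulator randomizes its pitch upward from a slightly lowered base.
# The cue plays at reduced volume with player footstep attenuation and overrides concurrency to cap simultaneous instances.
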SoundCue /CorruptionGameplay/Sounds/Weapons/Melee/Scythe/Scythe_PowerUp_BullRushSteps_Cue
root=Modulator_2
nodes [
 {id=Modulator_2, children [Random_1], properties {pitchMax=1.2, pitchMin=1, volumeMax=1, volumeMin=1}}
 {id=Random_1, children [WavePlayer_3,WavePlayer_4,WavePlayer_5,WavePlayer_6,WavePlayer_7], properties {weights [1,1,1,1,1]}}
 {id=WavePlayer_3, properties {soundWaveAssetPtr=/Game/Sounds/Fort_Step_Sounds/Constructor/Bullrush_Step_01}}
 {id=WavePlayer_4, properties {soundWaveAssetPtr=/Game/Sounds/Fort_Step_Sounds/Constructor/Bullrush_Step_02}}
 {id=WavePlayer_5, properties {soundWaveAssetPtr=/Game/Sounds/Fort_Step_Sounds/Constructor/Bullrush_Step_03}}
 {id=WavePlayer_6, properties {soundWaveAssetPtr=/Game/Sounds/Fort_Step_Sounds/Constructor/Bullrush_Step_04}}
 {id=WavePlayer_7, properties {soundWaveAssetPtr=/Game/Sounds/Fort_Step_Sounds/Constructor/Bullrush_Step_05}}
]
properties {
 attenuationSettings=/Game/Sounds/Attenuation/Fort_Player_Footsteps_Att
 bOverrideConcurrency=true
 maxDistance=3500
 pitchMultiplier=0.9
 soundClassObject=&Fort_WeaponMelee
 volumeMultiplier=0.3
 attenuationOverrides {falloffDistance=2700, lPFRadiusMax=3000, lPFRadiusMin=1600, attenuationShapeExtents {x=300}}
 concurrencyOverrides {maxCount=4}
 concurrencySet [null]
}
