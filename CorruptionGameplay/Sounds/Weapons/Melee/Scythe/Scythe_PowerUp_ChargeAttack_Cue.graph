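# Sound Cue for the scythe's power-up charge attack in CorruptionGameplay, routed to the melee weapon sound class.
# A Random node picks one of three equally weighted wave variants and feeds a Modulator that holds pitch fixed.
# Inline attenuation overrides take precedence over the linked scythe attenuation asset and use HRTF spatialization.
# Concurrency is capped by the Corruption dash global-limit setting.
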SoundCue /CorruptionGameplay/Sounds/Weapons/Melee/Scythe/Scythe_PowerUp_ChargeAttack_Cue
root=Modulator_0
nodes [
 {id=Modulator_0, children [Random_1], properties {pitchMax=1, pitchMin=1}}
 {id=Random_1, children [WavePlayer_2,WavePlayer_3,WavePlayer_4], properties {weights [1,1,1]}}
 {id=WavePlayer_2, properties {soundWaveAssetPtr=&/Scythe_Powerup_ChargeAttack_01}}
 {id=WavePlayer_3, properties {soundWaveAssetPtr=&/Scythe_Powerup_ChargeAttack_02}}
 {id=WavePlayer_4, properties {soundWaveAssetPtr=&/Scythe_Powerup_ChargeAttack_03}}
]
properties {
 attenuationSettings=/Game/Sounds/Attenuation/Fort_Weapon_Melee_Scythe_Long_Linear_Att
 bOverrideAttenuation=true
 maxDistance=3200
 soundClassObject=&Fort_WeaponMelee_PSM
 volumeMultiplier=1
 attenuationOverrides {
  binauralRadius=500
  falloffDistance=2400
  nonSpatializedRadiusStart=500
  spatializationAlgorithm=SPATIALIZATION_HRTF
  stereoSpread=2000
  attenuationShapeExtents {x=800}
 }
 concurrencySet [/CorruptionGameplay/Sounds/Settings/Corruption_Dash_GlobalLimit_Conc]
}
