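# Sound cue for the impact of the scythe's powered-up charge attack, a player melee weapon sound.
# A short fixed delay feeds a modulator with pitch held constant, then a random node picks one of three equally weighted impact waves.
# It uses a long linear scythe attenuation asset, the melee weapon sound class and the player melee concurrency set.
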SoundCue /CorruptionGameplay/Sounds/Weapons/Melee/Scythe/Scythe_PowerUp_ChargeAttack_Impact_Cue
root=Delay_0
nodes [
 {id=Delay_0, children [Modulator_1], properties {delayMax=0.1, delayMin=0.1}}
 {id=Modulator_1, children [Random_1], properties {pitchMax=1, pitchMin=1}}
 {id=Random_1, children [WavePlayer_3,WavePlayer_4,WavePlayer_5], properties {weights [1,1,1]}}
 {id=WavePlayer_3, properties {soundWaveAssetPtr=&/Scythe_Powerup_ChargeAttack_Impact_01}}
 {id=WavePlayer_4, properties {soundWaveAssetPtr=&/Scythe_Powerup_ChargeAttack_Impact_02}}
 {id=WavePlayer_5, properties {soundWaveAssetPtr=&/Scythe_Powerup_ChargeAttack_Impact_03}}
]
properties {
 attenuationSettings=/Game/Sounds/Attenuation/Fort_Weapon_Melee_Scythe_Long_Linear_Att
 maxDistance=2800
 soundClassObject=&Fort_WeaponMelee_PSM
 volumeMultiplier=0.9
 attenuationOverrides {bAttenuate=false}
 concurrencyOverrides {resolutionRule=PreventNew}
 concurrencySet [&Player_Melee_Weapon_Conc]
}
