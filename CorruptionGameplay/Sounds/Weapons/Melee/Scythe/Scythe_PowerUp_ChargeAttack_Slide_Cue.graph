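# Sound cue for the scythe power-up charge attack slide, a melee weapon sound in the Corruption gameplay content.
# A Random node picks one of three equally weighted slide waves and feeds a Modulator that holds pitch fixed.
# The cue plays at reduced volume in a melee weapon sound class, with long linear scythe attenuation and HRTF set in its attenuation overrides.
# A dash global-limit concurrency setting caps simultaneous instances.
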SoundCue /CorruptionGameplay/Sounds/Weapons/Melee/Scythe/Scythe_PowerUp_ChargeAttack_Slide_Cue
root=Modulator_0
nodes [
 {id=Modulator_0, children [Random_0], properties {pitchMax=1, pitchMin=1}}
 {id=Random_0, children [WavePlayer_0,WavePlayer_1,WavePlayer_5], properties {weights [1,1,1]}}
 {id=WavePlayer_0, properties {soundWaveAssetPtr=&/Scythe_Powerup_ChargeAttack_Slide_01}}
 {id=WavePlayer_1, properties {soundWaveAssetPtr=&/Scythe_Powerup_ChargeAttack_Slide_02}}
 {id=WavePlayer_5, properties {soundWaveAssetPtr=&/Scythe_Powerup_ChargeAttack_Slide_03}}
]
properties {
 attenuationSettings=/Game/Sounds/Attenuation/Fort_Weapon_Melee_Scythe_Long_Linear_Att
 maxDistance=2800
 soundClassObject=&Fort_WeaponMelee_PSM
 volumeMultiplier=0.5
 attenuationOverrides {spatializationAlgorithm=SPATIALIZATION_HRTF}
 concurrencySet [/CorruptionGameplay/Sounds/Settings/Corruption_Dash_GlobalLimit_Conc]
}
