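# Sound cue for dropping the Scythe power-up melee weapon.
# A Random node picks one of three drop waves and a Modulator adds slight upward pitch variation.
# A Mixer layers that result with a fixed, quieter swing impact layer wave.
# Playback uses weapon-drop attenuation, the loot-collect sound class and a weapon-drop concurrency set named for stopping the oldest instance.
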SoundCue /CorruptionGameplay/Sounds/Weapons/Melee/Scythe/Scythe_PowerUp_Drop_Cue
root=Mixer_0
nodes [
 {id=Mixer_0, children [Modulator_0,WavePlayer_1], properties {inputVolume [1.2,0.5]}}
 {id=Modulator_0, children [Random_1], properties {pitchMax=1.1, pitchMin=1}}
 {id=Random_1, children [WavePlayer_0,WavePlayer_5,WavePlayer_6], properties {weights [1,1,1]}}
 {id=WavePlayer_0, properties {soundWaveAssetPtr=&/Scythe_Powerup_Drop_01}}
 {id=WavePlayer_1, properties {soundWaveAssetPtr=&/Scythe_Powerup_Swing_ImpactLayer_05}}
 {id=WavePlayer_5, properties {soundWaveAssetPtr=&/Scythe_Powerup_Drop_02}}
 {id=WavePlayer_6, properties {soundWaveAssetPtr=&/Scythe_Powerup_Drop_03}}
]
properties {
 attenuationSettings=&Fort_Weapon_Drop_Att
 maxDistance=3000
 pitchMultiplier=1.1
 soundClassObject=&Fort_Loot_Collect
 volumeMultiplier=0.5
 concurrencyOverrides {resolutionRule=PreventNew}
 concurrencySet [&Weapon_Drop_StopOldest]
}
