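# Sound cue for equipping the powered-up scythe melee weapon.
# A Random node picks one of three equally weighted equip wave variations.
# The cue plays slightly pitched up in the player foley sound class, using long linear scythe attenuation.
# It overrides concurrency so new instances are prevented once the limit is reached.
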SoundCue /CorruptionGameplay/Sounds/Weapons/Melee/Scythe/Scythe_PowerUp_Equip_Cue
root=Random_1
nodes [
 {id=Random_1, children [WavePlayer_0,WavePlayer_1,WavePlayer_2], properties {weights [1,1,1]}}
 {id=WavePlayer_0, properties {soundWaveAssetPtr=&/Scythe_Powerup_Equip_01}}
 {id=WavePlayer_1, properties {soundWaveAssetPtr=&/Scythe_Powerup_Equip_02}}
 {id=WavePlayer_2, properties {soundWaveAssetPtr=&/Scythe_Powerup_Equip_03}}
]
properties {
 attenuationSettings=/Game/Sounds/Attenuation/Fort_Weapon_Melee_Scythe_Long_Linear_Att
 bOverrideConcurrency=true
 maxDistance=2800
 pitchMultiplier=1.1
 soundClassObject=&Fort_Player_Foley
 volumeMultiplier=1
 concurrencyOverrides {resolutionRule=PreventNew}
}
