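# Sound cue for the start of the scythe's powered-up jump attack, in the melee weapon sound class.
# A single wave player feeds a modulator left at its default settings, with a slight pitch raise on the cue.
# Inline attenuation overrides, used in place of the assigned explosion attenuation asset, set HRTF spatialization and falloff.
# A player melee weapon concurrency set governs simultaneous playback.
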SoundCue /CorruptionGameplay/Sounds/Weapons/Melee/Scythe/Scythe_PowerUp_Jump_Attack_Start_Cue
root=Modulator_1
nodes [
 {id=Modulator_1, children [WavePlayer_0]}
 {id=WavePlayer_0, properties {soundWaveAssetPtr=&/Scythe_Powerup_JumpAttack_01}}
]
properties {
 attenuationSettings=/Game/Sounds/Attenuation/Fort_Explosion_Close_StW_Att
 bOverrideAttenuation=true
 maxDistance=3200
 pitchMultiplier=1.05
 soundClassObject=&Fort_WeaponMelee_PSM
 volumeMultiplier=1
 attenuationOverrides {
  binauralRadius=500
  falloffDistance=2400
  nonSpatializedRadiusStart=500
  spatializationAlgorithm=SPATIALIZATION_HRTF
  stereoSpread=2000
  attenuationShapeExtents {x=800}
 }
 concurrencySet [&Player_Melee_Weapon_Conc]
}
