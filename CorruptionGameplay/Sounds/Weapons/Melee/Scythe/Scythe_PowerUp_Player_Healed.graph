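# Sound cue for a player-healed power-up sound on the Corruption scythe melee weapon.
# It plays an Athena essential-oils completion wave after a short fixed delay, with a slight volume boost.
# It uses consumable-use attenuation and the consumables sound class, with a concurrency override that caps simultaneous instances and stops the oldest.
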SoundCue /CorruptionGameplay/Sounds/Weapons/Melee/Scythe/Scythe_PowerUp_Player_Healed
root=Delay_0
nodes [
 {id=Delay_0, children [WavePlayer_0], properties {delayMax=0.3, delayMin=0.3}}
 {id=WavePlayer_0, properties {soundWaveAssetPtr=/Game/Athena/Sounds/Items/sfx_EssentialOils_Complete_01}}
]
properties {
 attenuationSettings=&Fort_Consumable_Use_Athena_Att
 bOverrideConcurrency=true
 maxDistance=2250
 soundClassObject=&Fort_Consumables
 volumeMultiplier=1.2
 concurrencyOverrides {maxCount=4, resolutionRule=StopOldest}
}
