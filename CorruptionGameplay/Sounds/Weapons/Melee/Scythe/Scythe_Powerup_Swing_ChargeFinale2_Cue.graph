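# Sound cue for a scythe melee weapon, named for the charge-finale swing of its powerup.
# A single wave player is the root and plays Scythe_Powerup_Swing_ChargeFinale_01, with no randomization, mixing or modulation nodes.
# The cue uses long linear scythe attenuation, the melee weapon sound class and player melee weapon concurrency, with a slight volume boost.
# Despite the "Finale2" in the cue's name, the referenced wave carries the _01 suffix.
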SoundCue /CorruptionGameplay/Sounds/Weapons/Melee/Scythe/Scythe_Powerup_Swing_ChargeFinale2_Cue
root=WavePlayer_1
nodes [{id=WavePlayer_1, properties {soundWaveAssetPtr=&/Scythe_Powerup_Swing_ChargeFinale_01}}]
properties {
 attenuationSettings=/Game/Sounds/Attenuation/Fort_Weapon_Melee_Scythe_Long_Linear_Att
 maxDistance=2800
 soundClassObject=&Fort_WeaponMelee_PSM
 volumeMultiplier=1.1
 concurrencySet [&Player_Melee_Weapon_Conc]
}
