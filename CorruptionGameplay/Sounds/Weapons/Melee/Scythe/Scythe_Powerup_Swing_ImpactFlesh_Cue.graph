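# Sound cue for a powered-up scythe melee swing hitting flesh.
# A mixer layers a random flesh-impact wave with a random impact-layer wave, the latter at half volume after a short fixed delay.
# Each layer passes through its own modulator for pitch variation.
# The cue uses the melee weapon sound class, a long linear scythe attenuation asset and a shared player melee concurrency set.
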SoundCue /CorruptionGameplay/Sounds/Weapons/Melee/Scythe/Scythe_Powerup_Swing_ImpactFlesh_Cue
root=Mixer_1
nodes [
 {id=Delay_0, children [Modulator_0], properties {delayMax=0.1, delayMin=0.1}}
 {id=Mixer_1, children [Modulator_1,Delay_0], properties {inputVolume [1,0.5]}}
 {id=Modulator_0, children [Random_1], properties {pitchMax=1, pitchMin=1.2}}
 {id=Modulator_1, children [Random_0], properties {pitchMax=0.9, pitchMin=1.2}}
 {id=Random_0, children [WavePlayer_0,WavePlayer_1,WavePlayer_2], properties {weights [1,1,1]}}
 {id=Random_1, children [WavePlayer_3,WavePlayer_4,WavePlayer_5,WavePlayer_6,WavePlayer_7,WavePlayer_8], properties {weights [1,1,1,1,1,1]}}
 {id=WavePlayer_0, properties {soundWaveAssetPtr=&/Scythe_Powerup_Swing_ImpactFlesh_01}}
 {id=WavePlayer_1, properties {soundWaveAssetPtr=&/Scythe_Powerup_Swing_ImpactFlesh_02}}
 {id=WavePlayer_2, properties {soundWaveAssetPtr=&/Scythe_Powerup_Swing_ImpactFlesh_03}}
 {id=WavePlayer_3, properties {soundWaveAssetPtr=&/Scythe_Powerup_Swing_ImpactLayer_01}}
 {id=WavePlayer_4, properties {soundWaveAssetPtr=&/Scythe_Powerup_Swing_ImpactLayer_02}}
 {id=WavePlayer_5, properties {soundWaveAssetPtr=&/Scythe_Powerup_Swing_ImpactLayer_03}}
 {id=WavePlayer_6, properties {soundWaveAssetPtr=&/Scythe_Powerup_Swing_ImpactLayer_04}}
 {id=WavePlayer_7, properties {soundWaveAssetPtr=&/Scythe_Powerup_Swing_ImpactLayer_05}}
 {id=WavePlayer_8, properties {soundWaveAssetPtr=&/Scythe_Powerup_Swing_ImpactLayer_06}}
]
properties {
 attenuationSettings=/Game/Sounds/Attenuation/Fort_Weapon_Melee_Scythe_Long_Linear_Att
 maxDistance=2800
 soundClassObject=&Fort_WeaponMelee_PSM
 volumeMultiplier=0.8
 attenuationOverrides {bAttenuate=false}
 concurrencyOverrides {resolutionRule=PreventNew}
 concurrencySet [&Player_Melee_Weapon_Conc]
}
