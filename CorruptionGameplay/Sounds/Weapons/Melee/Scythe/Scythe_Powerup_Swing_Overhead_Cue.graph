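# Sound cue for the powered-up overhead swing of the scythe melee weapon.
# A Random node picks evenly among three swing wave variations, then feeds a Modulator for slight playback variation.
# Output uses a long linear scythe attenuation, the melee weapon sound class and a player melee concurrency group, with reduced volume.
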SoundCue /CorruptionGameplay/Sounds/Weapons/Melee/Scythe/Scythe_Powerup_Swing_Overhead_Cue
root=Modulator_14
nodes [
 {id=Modulator_14, children [Random_0]}
 {id=Random_0, children [WavePlayer_1,WavePlayer_2,WavePlayer_3], properties {weights [1,1,1]}}
 {id=WavePlayer_1, properties {soundWaveAssetPtr=&/Scythe_Powerup_Swing_Overhead_01}}
 {id=WavePlayer_2, properties {soundWaveAssetPtr=&/Scythe_Powerup_Swing_Overhead_02}}
 {id=WavePlayer_3, properties {soundWaveAssetPtr=&/Scythe_Powerup_Swing_Overhead_03}}
]
properties {
 attenuationSettings=/Game/Sounds/Attenuation/Fort_Weapon_Melee_Scythe_Long_Linear_Att
 maxDistance=2800
 soundClassObject=&Fort_WeaponMelee_PSM
 volumeMultiplier=0.7
 concurrencySet [&Player_Melee_Weapon_Conc]
}
